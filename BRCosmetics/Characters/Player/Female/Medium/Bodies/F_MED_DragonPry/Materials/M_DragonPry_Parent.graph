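# Parent material for the DragonPry character body, adding color variants for clothes, accents, dragon skin and glowing eyes on top of the shared 2023 character material.
# PreFX output is split into attributes, regraded with tint, power and gain controls, then blended by ColorCustomizationMask channels.
# Dragon skin mixes two tints using TechArtMask red, and eye emissive is hue-shifted through HSV.
# Enable scalars gate variants before PostFX.
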Material /BRCosmetics/Characters/Player/Female/Medium/Bodies/F_MED_DragonPry/Materials/M_DragonPry_Parent
graph {
 expressions {
  MaterialExpressionAdd_3 {
   class=MaterialExpressionAdd
   inputs [
    A {from=MaterialExpressionNamedRerouteUsage_49}
    B {from=MaterialExpressionVectorParameter_5, output=Metallic Offset}
   ]
  }
  MaterialExpressionAdd_4 {
   class=MaterialExpressionAdd
   inputs [
    A {from=MaterialExpressionNamedRerouteUsage_57}
    B {from=MaterialExpressionVectorParameter_13, output=Metallic Offset}
   ]
  }
  MaterialExpressionAdd_5 {
   class=MaterialExpressionAdd
   inputs [
    A {from=MaterialExpressionScalarParameter_33}
    B {from=MaterialExpressionConvert_10}
   ]
  }
  MaterialExpressionBlendMaterialAttributes_10 {
   class=MaterialExpressionBlendMaterialAttributes
   inputs [
    A {from=MaterialExpressionBlendMaterialAttributes_8, output=""}
    B {from=MaterialExpressionSetMaterialAttributes_10, output=""}
    Alpha {from=MaterialExpressionScalarParameter_35, output=""}
   ]
   properties {}
  }
  MaterialExpressionBlendMaterialAttributes_11 {
   class=MaterialExpressionBlendMaterialAttributes
   inputs [
    A {from=MaterialExpressionBlendMaterialAttributes_8, output=""}
    B {from=MaterialExpressionNamedRerouteUsage_75, output=""}
    Alpha {from=MaterialExpressionNamedRerouteUsage_76, output=""}
   ]
   properties {}
  }
  MaterialExpressionBlendMaterialAttributes_6 {
   class=MaterialExpressionBlendMaterialAttributes
   inputs [
    A {from=MaterialExpressionNamedRerouteUsage_66, output=""}
    B {from=MaterialExpressionNamedRerouteUsage_63, output=""}
    Alpha {from=MaterialExpressionNamedRerouteUsage_65, output=""}
   ]
   properties {}
  }
  MaterialExpressionBlendMaterialAttributes_7 {
   class=MaterialExpressionBlendMaterialAttributes
   inputs [
    A {from=MaterialExpressionBlendMaterialAttributes_6, output=""}
    B {from=MaterialExpressionNamedRerouteUsage_64, output=""}
    Alpha {from=MaterialExpressionNamedRerouteUsage_67, output=""}
   ]
   properties {}
  }
  MaterialExpressionBlendMaterialAttributes_8 {
   class=MaterialExpressionBlendMaterialAttributes
   inputs [
    A {from=MaterialExpressionNamedRerouteUsage_66, output=""}
    B {from=MaterialExpressionBlendMaterialAttributes_7, output=""}
    Alpha {from=MaterialExpressionScalarParameter_29, output=""}
   ]
   properties {}
  }
  MaterialExpressionConvert_10 {
   class=MaterialExpressionConvert
   inputs [None {from=MaterialExpressionRgbToHsv_0}]
   properties {
    nodeName=Break Float3
    convertInputs [{
      type=Vector3
      expressionInput {
       expression=&self:MaterialExpressionRgbToHsv_0
       inputName=None
       mask=0
       maskA=0
       maskB=0
       maskG=0
       maskR=0
       outputIndex=0
      }
     }]
    convertOutputs [{type=Scalar}, {type=Scalar}, {type=Scalar}]
   }
  }
  MaterialExpressionConvert_11 {
   class=MaterialExpressionConvert
   inputs [
    None {from=MaterialExpressionFrac_0}
    None {from=MaterialExpressionConvert_10}
    None {from=MaterialExpressionConvert_10}
   ]
   properties {
    nodeName=Make Vector3
    convertInputs [
     {
      type=Scalar
      expressionInput {
       expression=&self:MaterialExpressionFrac_0
       inputName=None
       mask=0
       maskA=0
       maskB=0
       maskG=0
       maskR=0
       outputIndex=0
      }
     }
     {
      type=Scalar
      expressionInput {
       expression=&self:MaterialExpressionConvert_10
       inputName=None
       mask=0
       maskA=0
       maskB=0
       maskG=0
       maskR=0
       outputIndex=1
      }
     }
     {
      type=Scalar
      expressionInput {
       expression=&self:MaterialExpressionConvert_10
       inputName=None
       mask=0
       maskA=0
       maskB=0
       maskG=0
       maskR=0
       outputIndex=2
      }
     }
    ]
    convertOutputs [{type=Vector3}]
   }
  }
  MaterialExpressionDesaturation_0 {
   class=MaterialExpressionDesaturation
   inputs [
    None {from=MaterialExpressionNamedRerouteDeclaration_9, output=""}
    Fraction {from="", output=""}
   ]
   properties {}
  }
  MaterialExpressionFrac_0 {class=MaterialExpressionFrac, inputs [None {from=MaterialExpressionAdd_5, output=""}], properties {}}
  MaterialExpressionGetMaterialAttributes_0 {
   class=MaterialExpressionGetMaterialAttributes
   inputs [None {from=MaterialExpressionNamedRerouteUsage_72, output=""}]
   properties {attributeGetTypes [
     #1
     #2
     #3
     #4
     #5
     #6
    ]}
  }
  MaterialExpressionHsvToRgb_0 {class=MaterialExpressionHsvToRgb, inputs [None {from=MaterialExpressionConvert_11, output=""}], properties {}}
  MaterialExpressionLinearInterpolate_6 {
   class=MaterialExpressionLinearInterpolate
   inputs [
    A {from=MaterialExpressionNamedRerouteUsage_1}
    B {from=MaterialExpressionNamedRerouteUsage_2}
    Alpha {from=MaterialExpressionNamedRerouteUsage_0}
   ]
  }
  MaterialExpressionMaterialFunctionCall_0 {
   class=MaterialExpressionMaterialFunctionCall
   inputs [
    Overide Texture Set UVs {from=""}
    Override Tech Art Mask UVs {from=""}
    Override Basic Emissive UVs {from=""}
   ]
   properties {materialFunction=&MF_CharacterParent_2023_PreFX}
  }
  MaterialExpressionMaterialFunctionCall_1 {
   class=MaterialExpressionMaterialFunctionCall
   inputs [
    Material {from=MaterialExpressionBlendMaterialAttributes_10}
    M {from=MaterialExpressionMaterialFunctionCall_0, output=M}
    Pre Skinned Local Position {from=MaterialExpressionMaterialFunctionCall_0, output=M}
    Distance Blend {from=MaterialExpressionMaterialFunctionCall_0, output=M}
    Fresnel {from=MaterialExpressionMaterialFunctionCall_0, output=M}
   ]
   properties {materialFunction=&MF_CharacterParent_2023_PostFX}
  }
  MaterialExpressionMaterialFunctionCall_3 {
   class=MaterialExpressionMaterialFunctionCall
   inputs [
    Normal {from=MaterialExpressionNamedRerouteUsage_55}
    Flatness {from=MaterialExpressionScalarParameter_24}
   ]
   properties {materialFunction=&FlattenNormal}
  }
  MaterialExpressionMaterialFunctionCall_4 {
   class=MaterialExpressionMaterialFunctionCall
   inputs [
    Normal {from=MaterialExpressionNamedRerouteUsage_61}
    Flatness {from=MaterialExpressionScalarParameter_27}
   ]
   properties {materialFunction=&FlattenNormal}
  }
  MaterialExpressionMultiply_13 {
   class=MaterialExpressionMultiply
   inputs [
    A {from=MaterialExpressionAdd_3}
    B {from=MaterialExpressionVectorParameter_5, output=Metallic Gain}
   ]
  }
  MaterialExpressionMultiply_14 {
   class=MaterialExpressionMultiply
   inputs [
    A {from=MaterialExpressionPower_7}
    B {from=MaterialExpressionReroute_12}
   ]
  }
  MaterialExpressionMultiply_15 {
   class=MaterialExpressionMultiply
   inputs [
    A {from=MaterialExpressionPower_8}
    B {from=MaterialExpressionVectorParameter_7, output=Roughness Gain}
   ]
  }
  MaterialExpressionMultiply_16 {
   class=MaterialExpressionMultiply
   inputs [
    A {from=MaterialExpressionVectorParameter_12, output=RGB}
    B {from=MaterialExpressionPower_9}
   ]
  }
  MaterialExpressionMultiply_17 {
   class=MaterialExpressionMultiply
   inputs [
    A {from=MaterialExpressionReroute_13}
    B {from=MaterialExpressionMultiply_16}
   ]
  }
  MaterialExpressionMultiply_18 {
   class=MaterialExpressionMultiply
   inputs [
    A {from=MaterialExpressionAdd_4}
    B {from=MaterialExpressionVectorParameter_13, output=Metallic Gain}
   ]
  }
  MaterialExpressionMultiply_19 {
   class=MaterialExpressionMultiply
   inputs [
    A {from=MaterialExpressionPower_11}
    B {from=MaterialExpressionReroute_15}
   ]
  }
  MaterialExpressionMultiply_20 {
   class=MaterialExpressionMultiply
   inputs [
    A {from=MaterialExpressionPower_12}
    B {from=MaterialExpressionVectorParameter_14, output=Roughness Gain}
   ]
  }
  MaterialExpressionMultiply_21 {
   class=MaterialExpressionMultiply
   inputs [
    A {from=MaterialExpressionVectorParameter_15, output=RGB}
    B {from=MaterialExpressionPower_13}
   ]
  }
  MaterialExpressionMultiply_22 {
   class=MaterialExpressionMultiply
   inputs [
    A {from=MaterialExpressionReroute_16}
    B {from=MaterialExpressionMultiply_21}
   ]
  }
  MaterialExpressionMultiply_23 {
   class=MaterialExpressionMultiply
   inputs [
    A {from=MaterialExpressionVectorParameter_16, output=RGB}
    B {from=MaterialExpressionPower_14}
   ]
  }
  MaterialExpressionMultiply_28 {
   class=MaterialExpressionMultiply
   inputs [
    A {from=MaterialExpressionVectorParameter_19, output=RGB}
    B {from=MaterialExpressionPower_18}
   ]
  }
  MaterialExpressionNamedRerouteDeclaration_0 {
   class=MaterialExpressionNamedRerouteDeclaration
   inputs [None {from=MaterialExpressionTextureSampleParameter2D_1, output=R}]
   properties {name="Dragon Skin Color 2 Mask (S)", nodeColor=<rgba 1 0 0.5859 1>}
  }
  MaterialExpressionNamedRerouteDeclaration_1 {
   class=MaterialExpressionNamedRerouteDeclaration
   inputs [None {from=MaterialExpressionSaturate_14}]
   properties {name="Dragon Skin Color (V3)", nodeColor=<rgba 0.3359 0 1 1>}
  }
  MaterialExpressionNamedRerouteDeclaration_17 {
   class=MaterialExpressionNamedRerouteDeclaration
   inputs [None {from=MaterialExpressionGetMaterialAttributes_0, output=Metallic}]
   properties {name="Metallic In (S)", nodeColor=<rgba 0.3826 1 0 1>}
  }
  MaterialExpressionNamedRerouteDeclaration_18 {
   class=MaterialExpressionNamedRerouteDeclaration
   inputs [None {from=MaterialExpressionGetMaterialAttributes_0, output=Specular}]
   properties {name="Specular In (S)", nodeColor=<rgba 0.3826 1 0 1>}
  }
  MaterialExpressionNamedRerouteDeclaration_19 {
   class=MaterialExpressionNamedRerouteDeclaration
   inputs [None {from=MaterialExpressionGetMaterialAttributes_0, output=Roughness}]
   properties {name="Roughness In (S)", nodeColor=<rgba 0.3826 1 0 1>}
  }
  MaterialExpressionNamedRerouteDeclaration_2 {
   class=MaterialExpressionNamedRerouteDeclaration
   inputs [None {from=MaterialExpressionSaturate_15}]
   properties {name="Dragon Skin Color 2 (V3)", nodeColor=<rgba 0.3359 0 1 1>}
  }
  MaterialExpressionNamedRerouteDeclaration_23 {
   class=MaterialExpressionNamedRerouteDeclaration
   inputs [None {from=MaterialExpressionMaterialFunctionCall_0, output=Result}]
   properties {name="Material In (MA)", nodeColor=<rgba 0.1719 1 0 1>}
  }
  MaterialExpressionNamedRerouteDeclaration_24 {
   class=MaterialExpressionNamedRerouteDeclaration
   inputs [None {from=MaterialExpressionSaturate_9}]
   properties {name="Accent BaseColor (V3)", nodeColor=<rgba 0.3359 0 1 1>}
  }
  MaterialExpressionNamedRerouteDeclaration_25 {
   class=MaterialExpressionNamedRerouteDeclaration
   inputs [None {from=MaterialExpressionSaturate_7}]
   properties {name="Accent Specular (S)", nodeColor=<rgba 1 0.1406 0 1>}
  }
  MaterialExpressionNamedRerouteDeclaration_26 {
   class=MaterialExpressionNamedRerouteDeclaration
   inputs [None {from=MaterialExpressionSaturate_8}]
   properties {name="Accent Roughness (S)", nodeColor=<rgba 0 0.1797 1 1>}
  }
  MaterialExpressionNamedRerouteDeclaration_27 {
   class=MaterialExpressionNamedRerouteDeclaration
   inputs [None {from=MaterialExpressionPower_6}]
   properties {name="Accent Metallic (S)", nodeColor=<rgba 0.5 1 0 1>}
  }
  MaterialExpressionNamedRerouteDeclaration_28 {
   class=MaterialExpressionNamedRerouteDeclaration
   inputs [None {from=MaterialExpressionGetMaterialAttributes_0, output=Normal}]
   properties {name="Normal In (S)", nodeColor=<rgba 0.3826 1 0 1>}
  }
  MaterialExpressionNamedRerouteDeclaration_29 {
   class=MaterialExpressionNamedRerouteDeclaration
   inputs [None {from=MaterialExpressionMaterialFunctionCall_3, output=Result}]
   properties {name="Accent Normal (V3)", nodeColor=<rgba 0 1 0.8594 1>}
  }
  MaterialExpressionNamedRerouteDeclaration_30 {
   class=MaterialExpressionNamedRerouteDeclaration
   inputs [None {from=MaterialExpressionSetMaterialAttributes_7}]
   properties {name="Accent (MA)", nodeColor=<rgba 1 0.5391 0 1>}
  }
  MaterialExpressionNamedRerouteDeclaration_31 {
   class=MaterialExpressionNamedRerouteDeclaration
   inputs [None {from=MaterialExpressionSaturate_13}]
   properties {name="Clothes BaseColor (V3)", nodeColor=<rgba 0.3359 0 1 1>}
  }
  MaterialExpressionNamedRerouteDeclaration_32 {
   class=MaterialExpressionNamedRerouteDeclaration
   inputs [None {from=MaterialExpressionSaturate_11}]
   properties {name="Clothes Specular (S)", nodeColor=<rgba 1 0.1406 0 1>}
  }
  MaterialExpressionNamedRerouteDeclaration_33 {
   class=MaterialExpressionNamedRerouteDeclaration
   inputs [None {from=MaterialExpressionSaturate_12}]
   properties {name="Clothes Roughness (S)", nodeColor=<rgba 0 0.1797 1 1>}
  }
  MaterialExpressionNamedRerouteDeclaration_34 {
   class=MaterialExpressionNamedRerouteDeclaration
   inputs [None {from=MaterialExpressionPower_10}]
   properties {name="Clothes Metallic (S)", nodeColor=<rgba 0.5 1 0 1>}
  }
  MaterialExpressionNamedRerouteDeclaration_35 {
   class=MaterialExpressionNamedRerouteDeclaration
   inputs [None {from=MaterialExpressionMaterialFunctionCall_4, output=Result}]
   properties {name="Clothes Normal (V3)", nodeColor=<rgba 0 1 0.8594 1>}
  }
  MaterialExpressionNamedRerouteDeclaration_36 {
   class=MaterialExpressionNamedRerouteDeclaration
   inputs [None {from=MaterialExpressionSetMaterialAttributes_8}]
   properties {name="Clothes (MA)", nodeColor=<rgba 1 0.5391 0 1>}
  }
  MaterialExpressionNamedRerouteDeclaration_38 {
   class=MaterialExpressionNamedRerouteDeclaration
   inputs [None {from=MaterialExpressionSetMaterialAttributes_9}]
   properties {name="Skin (MA)", nodeColor=<rgba 1 0.5391 0 1>}
  }
  MaterialExpressionNamedRerouteDeclaration_39 {
   class=MaterialExpressionNamedRerouteDeclaration
   inputs [None {from=MaterialExpressionGetMaterialAttributes_0, output=Emissive Color}]
   properties {name="Emission In (S)", nodeColor=<rgba 0.3826 1 0 1>}
  }
  MaterialExpressionNamedRerouteDeclaration_4 {
   class=MaterialExpressionNamedRerouteDeclaration
   inputs [None {from=MaterialExpressionTextureSampleParameter2D_0, output=R}]
   properties {name="Clothes Mask (S)", nodeColor=<rgba 1 0 0.5859 1>}
  }
  MaterialExpressionNamedRerouteDeclaration_40 {
   class=MaterialExpressionNamedRerouteDeclaration
   inputs [None {from=MaterialExpressionHsvToRgb_0}]
   properties {name="Eyes Emission (V3)", nodeColor=<rgba 0.5 1 0 1>}
  }
  MaterialExpressionNamedRerouteDeclaration_5 {
   class=MaterialExpressionNamedRerouteDeclaration
   inputs [None {from=MaterialExpressionTextureSampleParameter2D_0, output=G}]
   properties {name="Skin Mask (S)", nodeColor=<rgba 1 0 0.5859 1>}
  }
  MaterialExpressionNamedRerouteDeclaration_6 {
   class=MaterialExpressionNamedRerouteDeclaration
   inputs [None {from=MaterialExpressionTextureSampleParameter2D_0, output=B}]
   properties {name="Accent Mask (S)", nodeColor=<rgba 1 0 0.5859 1>}
  }
  MaterialExpressionNamedRerouteDeclaration_8 {
   class=MaterialExpressionNamedRerouteDeclaration
   inputs [None {from=MaterialExpressionDesaturation_0}]
   properties {name="Base Color In Desat (V3)", nodeColor=<rgba 0.3826 1 0 1>}
  }
  MaterialExpressionNamedRerouteDeclaration_9 {
   class=MaterialExpressionNamedRerouteDeclaration
   inputs [None {from=MaterialExpressionGetMaterialAttributes_0, output=Base Color}]
   properties {name="Base Color In (V3)", nodeColor=<rgba 0.3826 1 0 1>}
  }
  MaterialExpressionNamedRerouteUsage_0 {class=MaterialExpressionNamedRerouteUsage}
  MaterialExpressionNamedRerouteUsage_1 {class=MaterialExpressionNamedRerouteUsage}
  MaterialExpressionNamedRerouteUsage_2 {class=MaterialExpressionNamedRerouteUsage}
  MaterialExpressionNamedRerouteUsage_49 {class=MaterialExpressionNamedRerouteUsage}
  MaterialExpressionNamedRerouteUsage_50 {class=MaterialExpressionNamedRerouteUsage}
  MaterialExpressionNamedRerouteUsage_51 {class=MaterialExpressionNamedRerouteUsage}
  MaterialExpressionNamedRerouteUsage_54 {class=MaterialExpressionNamedRerouteUsage}
  MaterialExpressionNamedRerouteUsage_55 {class=MaterialExpressionNamedRerouteUsage}
  MaterialExpressionNamedRerouteUsage_56 {class=MaterialExpressionNamedRerouteUsage}
  MaterialExpressionNamedRerouteUsage_57 {class=MaterialExpressionNamedRerouteUsage}
  MaterialExpressionNamedRerouteUsage_58 {class=MaterialExpressionNamedRerouteUsage}
  MaterialExpressionNamedRerouteUsage_59 {class=MaterialExpressionNamedRerouteUsage}
  MaterialExpressionNamedRerouteUsage_60 {class=MaterialExpressionNamedRerouteUsage}
  MaterialExpressionNamedRerouteUsage_61 {class=MaterialExpressionNamedRerouteUsage}
  MaterialExpressionNamedRerouteUsage_62 {class=MaterialExpressionNamedRerouteUsage}
  MaterialExpressionNamedRerouteUsage_63 {class=MaterialExpressionNamedRerouteUsage}
  MaterialExpressionNamedRerouteUsage_64 {class=MaterialExpressionNamedRerouteUsage}
  MaterialExpressionNamedRerouteUsage_65 {class=MaterialExpressionNamedRerouteUsage}
  MaterialExpressionNamedRerouteUsage_66 {class=MaterialExpressionNamedRerouteUsage}
  MaterialExpressionNamedRerouteUsage_67 {class=MaterialExpressionNamedRerouteUsage}
  MaterialExpressionNamedRerouteUsage_68 {class=MaterialExpressionNamedRerouteUsage}
  MaterialExpressionNamedRerouteUsage_69 {class=MaterialExpressionNamedRerouteUsage}
  MaterialExpressionNamedRerouteUsage_72 {class=MaterialExpressionNamedRerouteUsage}
  MaterialExpressionNamedRerouteUsage_73 {class=MaterialExpressionNamedRerouteUsage}
  MaterialExpressionNamedRerouteUsage_74 {class=MaterialExpressionNamedRerouteUsage}
  MaterialExpressionNamedRerouteUsage_75 {class=MaterialExpressionNamedRerouteUsage}
  MaterialExpressionNamedRerouteUsage_76 {class=MaterialExpressionNamedRerouteUsage}
  MaterialExpressionNamedRerouteUsage_78 {class=MaterialExpressionNamedRerouteUsage}
  MaterialExpressionPower_10 {
   class=MaterialExpressionPower
   inputs [
    Base {from=MaterialExpressionSaturate_10}
    Exp {from=MaterialExpressionVectorParameter_13, output=Metallic Exp}
   ]
  }
  MaterialExpressionPower_11 {
   class=MaterialExpressionPower
   inputs [
    Base {from=MaterialExpressionNamedRerouteUsage_58}
    Exp {from=MaterialExpressionReroute_14}
   ]
  }
  MaterialExpressionPower_12 {
   class=MaterialExpressionPower
   inputs [
    Base {from=MaterialExpressionNamedRerouteUsage_59}
    Exp {from=MaterialExpressionVectorParameter_14, output=Roughness Exp}
   ]
  }
  MaterialExpressionPower_13 {
   class=MaterialExpressionPower
   inputs [
    Base {from=MaterialExpressionNamedRerouteUsage_60}
    Exp {from=MaterialExpressionReroute_19}
   ]
  }
  MaterialExpressionPower_14 {
   class=MaterialExpressionPower
   inputs [
    Base {from=MaterialExpressionNamedRerouteUsage_68}
    Exp {from=MaterialExpressionReroute_20}
   ]
  }
  MaterialExpressionPower_18 {
   class=MaterialExpressionPower
   inputs [
    Base {from=MaterialExpressionNamedRerouteUsage_78}
    Exp {from=MaterialExpressionReroute_23}
   ]
  }
  MaterialExpressionPower_6 {
   class=MaterialExpressionPower
   inputs [
    Base {from=MaterialExpressionSaturate_6}
    Exp {from=MaterialExpressionVectorParameter_5, output=Metallic Exp}
   ]
  }
  MaterialExpressionPower_7 {
   class=MaterialExpressionPower
   inputs [
    Base {from=MaterialExpressionNamedRerouteUsage_50}
    Exp {from=MaterialExpressionReroute_11}
   ]
  }
  MaterialExpressionPower_8 {
   class=MaterialExpressionPower
   inputs [
    Base {from=MaterialExpressionNamedRerouteUsage_51}
    Exp {from=MaterialExpressionVectorParameter_7, output=Roughness Exp}
   ]
  }
  MaterialExpressionPower_9 {
   class=MaterialExpressionPower
   inputs [
    Base {from=MaterialExpressionNamedRerouteUsage_54}
    Exp {from=MaterialExpressionReroute_18}
   ]
  }
  MaterialExpressionReroute_11 {class=MaterialExpressionReroute, inputs [None {from=MaterialExpressionVectorParameter_7, output=Specular Exp}]}
  MaterialExpressionReroute_12 {class=MaterialExpressionReroute, inputs [None {from=MaterialExpressionVectorParameter_7, output=Specular Gain}]}
  MaterialExpressionReroute_13 {class=MaterialExpressionReroute, inputs [None {from=MaterialExpressionVectorParameter_5, output=Albedo Gain}]}
  MaterialExpressionReroute_14 {class=MaterialExpressionReroute, inputs [None {from=MaterialExpressionVectorParameter_14, output=Specular Exp}]}
  MaterialExpressionReroute_15 {class=MaterialExpressionReroute, inputs [None {from=MaterialExpressionVectorParameter_14, output=Specular Gain}]}
  MaterialExpressionReroute_16 {class=MaterialExpressionReroute, inputs [None {from=MaterialExpressionVectorParameter_13, output=Albedo Gain}]}
  MaterialExpressionReroute_18 {class=MaterialExpressionReroute, inputs [None {from=MaterialExpressionVectorParameter_12, output=Exp}]}
  MaterialExpressionReroute_19 {class=MaterialExpressionReroute, inputs [None {from=MaterialExpressionVectorParameter_15, output=Exp}]}
  MaterialExpressionReroute_20 {class=MaterialExpressionReroute, inputs [None {from=MaterialExpressionVectorParameter_16, output=Exp}]}
  MaterialExpressionReroute_23 {class=MaterialExpressionReroute, inputs [None {from=MaterialExpressionVectorParameter_19, output=Exp}]}
  MaterialExpressionRgbToHsv_0 {class=MaterialExpressionRgbToHsv, inputs [None {from=MaterialExpressionNamedRerouteUsage_73, output=""}], properties {}}
  MaterialExpressionSaturate_10 {class=MaterialExpressionSaturate, inputs [None {from=MaterialExpressionMultiply_18}]}
  MaterialExpressionSaturate_11 {class=MaterialExpressionSaturate, inputs [None {from=MaterialExpressionMultiply_19}]}
  MaterialExpressionSaturate_12 {class=MaterialExpressionSaturate, inputs [None {from=MaterialExpressionMultiply_20}]}
  MaterialExpressionSaturate_13 {class=MaterialExpressionSaturate, inputs [None {from=MaterialExpressionMultiply_22}]}
  MaterialExpressionSaturate_14 {class=MaterialExpressionSaturate, inputs [None {from=MaterialExpressionMultiply_23}]}
  MaterialExpressionSaturate_15 {class=MaterialExpressionSaturate, inputs [None {from=MaterialExpressionMultiply_28}]}
  MaterialExpressionSaturate_6 {class=MaterialExpressionSaturate, inputs [None {from=MaterialExpressionMultiply_13}]}
  MaterialExpressionSaturate_7 {class=MaterialExpressionSaturate, inputs [None {from=MaterialExpressionMultiply_14}]}
  MaterialExpressionSaturate_8 {class=MaterialExpressionSaturate, inputs [None {from=MaterialExpressionMultiply_15}]}
  MaterialExpressionSaturate_9 {class=MaterialExpressionSaturate, inputs [None {from=MaterialExpressionMultiply_17}]}
  MaterialExpressionScalarParameter_24 {class=MaterialExpressionScalarParameter, properties {group=Color Variants, parameterName=Accent Normal Flatness, sortPriority=140}}
  MaterialExpressionScalarParameter_27 {class=MaterialExpressionScalarParameter, properties {group=Color Variants, parameterName=Clothes Normal Flatness, sortPriority=40}}
  MaterialExpressionScalarParameter_29 {
   class=MaterialExpressionScalarParameter
   properties {
    controlType=Enumeration
    enumeration=/Game/Global/Materials/Enums/E_Mat_OnOff
    group=Color Variants
    parameterName=Enable Clothes Color Variants
    sliderMax=1
    sortPriority=10
   }
  }
  MaterialExpressionScalarParameter_33 {class=MaterialExpressionScalarParameter, properties {group=Color Variants, parameterName=Dragon Eyes Hue Offset, sortPriority=350}}
  MaterialExpressionScalarParameter_35 {
   class=MaterialExpressionScalarParameter
   properties {
    controlType=Enumeration
    enumeration=&E_Mat_OnOff
    group=Color Variants
    parameterName=Enable Dragon Skin Color Variants
    sliderMax=1
    sortPriority=300
   }
  }
  MaterialExpressionSetMaterialAttributes_10 {
   class=MaterialExpressionSetMaterialAttributes
   inputs [
    MaterialAttributes {from=MaterialExpressionBlendMaterialAttributes_11, output=""}
    Emissive Color {from=MaterialExpressionNamedRerouteUsage_74, output=""}
   ]
   properties {attributeSetTypes [#6]}
  }
  MaterialExpressionSetMaterialAttributes_7 {
   class=MaterialExpressionSetMaterialAttributes
   inputs [
    MaterialAttributes {from=MaterialExpressionNamedRerouteUsage_56, output=""}
    Base Color {from=MaterialExpressionNamedRerouteDeclaration_24, output=""}
    Metallic {from=MaterialExpressionNamedRerouteDeclaration_27, output=""}
    Specular {from=MaterialExpressionNamedRerouteDeclaration_25, output=""}
    Roughness {from=MaterialExpressionNamedRerouteDeclaration_26, output=""}
    Normal {from=MaterialExpressionNamedRerouteDeclaration_29, output=""}
   ]
   properties {attributeSetTypes [
     #1
     #2
     #3
     #4
     #5
    ]}
  }
  MaterialExpressionSetMaterialAttributes_8 {
   class=MaterialExpressionSetMaterialAttributes
   inputs [
    MaterialAttributes {from=MaterialExpressionNamedRerouteUsage_62, output=""}
    Base Color {from=MaterialExpressionNamedRerouteDeclaration_31, output=""}
    Metallic {from=MaterialExpressionNamedRerouteDeclaration_34, output=""}
    Specular {from=MaterialExpressionNamedRerouteDeclaration_32, output=""}
    Roughness {from=MaterialExpressionNamedRerouteDeclaration_33, output=""}
    Normal {from=MaterialExpressionNamedRerouteDeclaration_35, output=""}
   ]
   properties {attributeSetTypes [
     #1
     #2
     #3
     #4
     #5
    ]}
  }
  MaterialExpressionSetMaterialAttributes_9 {
   class=MaterialExpressionSetMaterialAttributes
   inputs [
    MaterialAttributes {from=MaterialExpressionNamedRerouteUsage_69, output=""}
    Base Color {from=MaterialExpressionLinearInterpolate_6, output=""}
   ]
   properties {attributeSetTypes [#1]}
  }
  MaterialExpressionTextureSampleParameter2D_0 {
   class=MaterialExpressionTextureSampleParameter2D
   inputs [
    UVs {from="", output=""}
    Apply View MipBias {from="", output=""}
   ]
   properties {
    group=Color Variants
    parameterName=ColorCustomizationMask
    samplerType=SAMPLERTYPE_Masks
    sortPriority=0
    texture=&T_BlackMask
   }
  }
  MaterialExpressionTextureSampleParameter2D_1 {
   class=MaterialExpressionTextureSampleParameter2D
   inputs [
    UVs {from="", output=""}
    Apply View MipBias {from="", output=""}
   ]
   properties {
    group=05_Tech Art Masks
    parameterName=TechArtMask
    samplerSource=SSM_Wrap_WorldGroupSettings
    samplerType=SAMPLERTYPE_Masks
    sortPriority=0
    texture=&T_BlackMask
   }
  }
  MaterialExpressionVectorParameter_12 {
   class=MaterialExpressionVectorParameter
   properties {
    group=Color Variants
    parameterName=Accent Color
    sortPriority=130
    channelNames {a=Exp, b=Tint B, g=Tint G, r=Tint R}
    defaultValue=<rgba 1 1 1 1>
   }
  }
  MaterialExpressionVectorParameter_13 {
   class=MaterialExpressionVectorParameter
   properties {
    group=Color Variants
    parameterName=Clothes Metallic
    sortPriority=50
    channelNames {a=Albedo Gain, b=Metallic Exp, g=Metallic Gain, r=Metallic Offset}
    defaultValue=<rgba 0 1 1 1>
   }
  }
  MaterialExpressionVectorParameter_14 {
   class=MaterialExpressionVectorParameter
   properties {
    group=Color Variants
    parameterName=Clothes Specular
    sortPriority=60
    channelNames {a=Roughness Gain, b=Roughness Exp, g=Specular Gain, r=Specular Exp}
    defaultValue=<rgba 1 1 1 1>
   }
  }
  MaterialExpressionVectorParameter_15 {
   class=MaterialExpressionVectorParameter
   properties {
    group=Color Variants
    parameterName=Clothes Color
    sortPriority=30
    channelNames {a=Exp, b=Tint B, g=Tint G, r=Tint R}
    defaultValue=<rgba 1 1 1 1>
   }
  }
  MaterialExpressionVectorParameter_16 {
   class=MaterialExpressionVectorParameter
   properties {
    group=Color Variants
    parameterName=Dragon Skin Color
    sortPriority=320
    channelNames {a=Exp, b=Tint B, g=Tint G, r=Tint R}
    defaultValue=<rgba 1 1 1 1>
   }
  }
  MaterialExpressionVectorParameter_19 {
   class=MaterialExpressionVectorParameter
   properties {
    group=Color Variants
    parameterName=Dragon Skin Color 2
    sortPriority=325
    channelNames {a=Exp, b=Tint B, g=Tint G, r=Tint R}
    defaultValue=<rgba 1 1 1 1>
   }
  }
  MaterialExpressionVectorParameter_5 {
   class=MaterialExpressionVectorParameter
   properties {
    group=Color Variants
    parameterName=Accent Metallic
    sortPriority=150
    channelNames {a=Albedo Gain, b=Metallic Exp, g=Metallic Gain, r=Metallic Offset}
    defaultValue=<rgba 0 1 1 1>
   }
  }
  MaterialExpressionVectorParameter_7 {
   class=MaterialExpressionVectorParameter
   properties {
    group=Color Variants
    parameterName=Accent Specular
    sortPriority=160
    channelNames {a=Roughness Gain, b=Roughness Exp, g=Specular Gain, r=Specular Exp}
    defaultValue=<rgba 1 1 1 1>
   }
  }
 }
 properties {
  bAutomaticallySetUsageInEditor=false
  bUseMaterialAttributes=true
  bUsedWithClothing=true
  bUsedWithMeshDeformer=true
  bUsedWithMorphTargets=true
  bUsedWithNanite=true
  bUsedWithNiagaraMeshParticles=true
  bUsedWithSkeletalMesh=true
  pixelDepthOffsetMode=PDOM_Legacy
  shadingModel=MSM_FromMaterialExpression
  usedShadingModels=Default Lit | Subsurface | Clear Coat | Subsurface Profile | Cloth
 }
}
