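# Physics asset for the front cloth of the female medium DragonPry body, apparently for secondary motion.
# Five dyn cloth capsules chain from a base bone on the pelvis, with zero twist and narrow swing on one axis, wider on the other.
# Capsules on legs, feet, pelvis and spine form body shapes with free angular constraints, alongside a very large offset root capsule.
# A sword capsule also hangs from the pelvis under tight angular limits.
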PhysicsAsset /BRCosmetics/Characters/Player/Female/Medium/Bodies/F_MED_DragonPry/Meshes/F_MED_DragonPry_PhysicsFrontCloth_FE
bodies {
 ball_l [{shapeType=Capsule, length=9.9498, radius=8.2636, shapeName=ball_l_capsule, center=<-4 0 0>, rotation=<rot 90 180 0>}]
 ball_r [{shapeType=Capsule, length=9.9498, radius=8.2636, shapeName=ball_r_capsule, center=<4 0 0>, rotation=<rot 90 0 0>}]
 base_front_cloth_mid_01 [{shapeType=Capsule, length=2.5, radius=1, shapeName=base_front_cloth_mid_01_capsule, center=<4 0 3>, rotation=<rot 0 0 90>}]
 calf_l [{shapeType=Capsule, length=45, radius=8, shapeName=calf_l_capsule, center=<-22 -1.3 -0.5>, rotation=<rot -0.8 93 -88>}]
 calf_r [{shapeType=Capsule, length=45, radius=8, shapeName=calf_r_capsule, center=<22 1.3 0.5>, rotation=<rot -0.8 93 92>}]
 dyn_front_cloth_mid_01 [{shapeType=Capsule, length=15, radius=6, shapeName=dyn_front_cloth_mid_01_capsule, center=<4 0 4>, rotation=<rot 0 0 90>}]
 dyn_front_cloth_mid_02 [{shapeType=Capsule, length=15, radius=6, shapeName=dyn_front_cloth_mid_02_capsule, center=<4 0 5>, rotation=<rot 0 0 90>}]
 dyn_front_cloth_mid_03 [{shapeType=Capsule, length=10, radius=6, shapeName=dyn_front_cloth_mid_03_capsule, center=<4 0 4>, rotation=<rot 0 0 90>}]
 dyn_front_cloth_mid_04 [{shapeType=Capsule, length=8, radius=6, shapeName=dyn_front_cloth_mid_04_capsule, center=<4 0 4>, rotation=<rot 0 0 90>}]
 dyn_front_cloth_mid_05 [{shapeType=Capsule, length=6, radius=6, shapeName=dyn_front_cloth_mid_05_capsule, center=<4 0 3>, rotation=<rot 0 0 90>}]
 dyn_sword_weapon_01 [{
   shapeType=Capsule
   length=72.6079
   radius=4.4081
   shapeName=dyn_sword_weapon_01_capsule
   center=<16.1907 0.9376 0.1314>
   rotation=<rot 0 92.1783 -89.6297>
  }]
 pelvis [
  {shapeType=Capsule, length=1.7352, radius=10.7648, shapeName=pelvis_capsule, center=<-5.5704 -1.8 -4.9108>, rotation=<rot 82 180 0>}
  {shapeType=Capsule, length=1.7352, radius=10.7648, shapeName=pelvis_capsule, center=<-5.5188 -1.8 5.1237>, rotation=<rot 79.1578 -0.0001 0>}
 ]
 root [{shapeType=Capsule, length=100, radius=200, shapeName=root_capsule, center=<0 31.5359 -187.3554>, rotation=<rot 0 -90 90>}]
 spine_01 [{shapeType=Capsule, length=10.9678, radius=8.41, shapeName=pelvis_capsule, center=<0 -0.5 0>}]
 spine_04 [{shapeType=Capsule, length=6, radius=13, shapeName=spine_04_capsule, center=<3.7 3 0>, rotation=<rot 0 20 0>}]
 thigh_l [
  {shapeType=Capsule, length=16, radius=7.3, center=<-30 0.8736 -0.25>, rotation=<rot 88.5 0 -180>}
  {shapeType=Capsule, length=14, radius=8.8, shapeName=thigh_l_capsule, center=<-11.9 -1.12 0.56>, rotation=<rot 86 177 180>}
 ]
 thigh_r [
  {shapeType=Capsule, length=16, radius=7.3, center=<30 -0.8736 0.25>, rotation=<rot 88.5 0 0>}
  {shapeType=Capsule, length=14, radius=8.8, shapeName=thigh_r_capsule, center=<11.9 1.12 -0.56>, rotation=<rot 86 177 0>}
 ]
}
constraints [
 {
  bone1Name=thigh_l
  bone2Name=pelvis
  swing1LimitDegrees=45
  swing1Motion=Free
  swing2LimitDegrees=45
  swing2Motion=Free
  twistLimitDegrees=45
  twistMotion=Free
 }
 {
  bone1Name=thigh_r
  bone2Name=pelvis
  swing1LimitDegrees=45
  swing1Motion=Free
  swing2LimitDegrees=45
  swing2Motion=Free
  twistLimitDegrees=45
  twistMotion=Free
 }
 {
  bone1Name=spine_01
  bone2Name=pelvis
  swing1LimitDegrees=45
  swing1Motion=Free
  swing2LimitDegrees=45
  swing2Motion=Free
  twistLimitDegrees=45
  twistMotion=Free
 }
 {
  bone1Name=calf_r
  bone2Name=thigh_r
  swing1LimitDegrees=45
  swing1Motion=Free
  swing2LimitDegrees=45
  swing2Motion=Free
  twistLimitDegrees=45
  twistMotion=Free
 }
 {
  bone1Name=calf_l
  bone2Name=thigh_l
  swing1LimitDegrees=45
  swing1Motion=Free
  swing2LimitDegrees=45
  swing2Motion=Free
  twistLimitDegrees=45
  twistMotion=Free
 }
 {
  bone1Name=base_front_cloth_mid_01
  bone2Name=pelvis
  swing1LimitDegrees=45
  swing1Motion=Free
  swing2LimitDegrees=45
  swing2Motion=Free
  twistLimitDegrees=45
  twistMotion=Free
 }
 {bone1Name=dyn_front_cloth_mid_01, bone2Name=base_front_cloth_mid_01, swing1LimitDegrees=1, swing2LimitDegrees=20, twistLimitDegrees=0}
 {bone1Name=dyn_front_cloth_mid_02, bone2Name=dyn_front_cloth_mid_01, swing1LimitDegrees=1, swing2LimitDegrees=20, twistLimitDegrees=0}
 {bone1Name=dyn_front_cloth_mid_03, bone2Name=dyn_front_cloth_mid_02, swing1LimitDegrees=1, swing2LimitDegrees=20, twistLimitDegrees=0}
 {bone1Name=dyn_front_cloth_mid_04, bone2Name=dyn_front_cloth_mid_03, swing1LimitDegrees=1, swing2LimitDegrees=20, twistLimitDegrees=0}
 {bone1Name=dyn_front_cloth_mid_05, bone2Name=dyn_front_cloth_mid_04, swing1LimitDegrees=1, swing2LimitDegrees=20, twistLimitDegrees=0}
 {
  bone1Name=ball_l
  bone2Name=calf_l
  swing1LimitDegrees=45
  swing1Motion=Free
  swing2LimitDegrees=45
  swing2Motion=Free
  twistLimitDegrees=45
  twistMotion=Free
 }
 {
  bone1Name=ball_r
  bone2Name=calf_r
  swing1LimitDegrees=45
  swing1Motion=Free
  swing2LimitDegrees=45
  swing2Motion=Free
  twistLimitDegrees=45
  twistMotion=Free
 }
 {
  bone1Name=spine_04
  bone2Name=spine_01
  swing1LimitDegrees=45
  swing1Motion=Free
  swing2LimitDegrees=45
  swing2Motion=Free
  twistLimitDegrees=45
  twistMotion=Free
 }
 {
  bone1Name=pelvis
  bone2Name=root
  swing1LimitDegrees=45
  swing1Motion=Free
  swing2LimitDegrees=45
  swing2Motion=Free
  twistLimitDegrees=45
  twistMotion=Free
 }
 {bone1Name=dyn_sword_weapon_01, bone2Name=pelvis, swing1LimitDegrees=10, swing2LimitDegrees=10, twistLimitDegrees=0}
]
