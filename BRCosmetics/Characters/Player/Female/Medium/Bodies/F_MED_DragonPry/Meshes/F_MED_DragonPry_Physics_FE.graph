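# Physics asset for the female medium DragonPry body cosmetic, covering the lower body and a hanging back-cloth bone chain.
# Six dynamic cloth capsules link in sequence from a base body constrained to the pelvis, with narrow swing limits and zero twist.
# Capsules on the pelvis, lowest spine bone, thighs and calves approximate the body shape, joined by constraints with free swing and twist.
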PhysicsAsset /BRCosmetics/Characters/Player/Female/Medium/Bodies/F_MED_DragonPry/Meshes/F_MED_DragonPry_Physics_FE
bodies {
 base_back_cloth_01 [{shapeType=Capsule, length=10, radius=5, shapeName=base_back_cloth_01_capsule, center=<5 0 4>, rotation=<rot 0 0 90>}]
 calf_l [{shapeType=Capsule, length=49.7364, radius=8, shapeName=calf_l_capsule, center=<-17.3146 1.6296 0>, rotation=<rot 90 180 0>}]
 calf_r [{shapeType=Capsule, length=49.7364, radius=8, shapeName=calf_r_capsule, center=<17.3146 -1.3601 0>, rotation=<rot 90 0 0>}]
 dyn_back_cloth_01 [{shapeType=Capsule, length=10, radius=5, shapeName=dyn_back_cloth_01_capsule, center=<5 0 4>, rotation=<rot 0 0 90>}]
 dyn_back_cloth_02 [{shapeType=Capsule, length=10, radius=5, shapeName=dyn_back_cloth_01_capsule, center=<5 0 4>, rotation=<rot 0 0 90>}]
 dyn_back_cloth_03 [{shapeType=Capsule, length=10, radius=5, shapeName=dyn_back_cloth_01_capsule, center=<5 0 4>, rotation=<rot 0 0 90>}]
 dyn_back_cloth_04 [{shapeType=Capsule, length=10, radius=5, shapeName=dyn_back_cloth_01_capsule, center=<5 0 4>, rotation=<rot 0 0 90>}]
 dyn_back_cloth_05 [{shapeType=Capsule, length=10, radius=5, shapeName=dyn_back_cloth_01_capsule, center=<5 0 4>, rotation=<rot 0 0 90>}]
 dyn_back_cloth_06 [{shapeType=Capsule, length=10, radius=5, shapeName=dyn_back_cloth_01_capsule, center=<5 0 4>, rotation=<rot 0 0 90>}]
 pelvis [
  {shapeType=Capsule, length=1.7352, radius=10.7648, shapeName=pelvis_capsule, center=<-5.5704 -1.8 -4.9108>, rotation=<rot 82 180 0>}
  {shapeType=Capsule, length=1.7352, radius=10.7648, shapeName=pelvis_capsule, center=<-5.5188 -1.8 5.1237>, rotation=<rot 79.1578 -0.0001 0>}
 ]
 spine_01 [{shapeType=Capsule, length=9.243, radius=10.1348, shapeName=pelvis_capsule}]
 thigh_l [
  {shapeType=Capsule, length=9.3266, radius=8, center=<-20.0202 0.9221 0>, rotation=<rot 88.4676 179.9996 179.9997>}
  {
   shapeType=Capsule
   length=9.1973
   radius=9.5766
   shapeName=thigh_l_capsule
   center=<-6.8061 -1.3805 0.9012>
   rotation=<rot 86.135 177.2457 179.9999>
  }
 ]
 thigh_r [
  {shapeType=Capsule, length=9.3266, radius=8, center=<20.0202 -0.9221 0>, rotation=<rot 88.4676 179.9996 -0.0003>}
  {
   shapeType=Capsule
   length=9.1973
   radius=9.5766
   shapeName=thigh_r_capsule
   center=<6.8061 1.3805 -0.9012>
   rotation=<rot 86.135 177.2457 -0.0001>
  }
 ]
}
constraints [
 {
  bone1Name=thigh_l
  bone2Name=pelvis
  swing1LimitDegrees=45
  swing1Motion=Free
  swing2LimitDegrees=45
  swing2Motion=Free
  twistLimitDegrees=45
  twistMotion=Free
 }
 {
  bone1Name=thigh_r
  bone2Name=pelvis
  swing1LimitDegrees=45
  swing1Motion=Free
  swing2LimitDegrees=45
  swing2Motion=Free
  twistLimitDegrees=45
  twistMotion=Free
 }
 {
  bone1Name=spine_01
  bone2Name=pelvis
  swing1LimitDegrees=45
  swing1Motion=Free
  swing2LimitDegrees=45
  swing2Motion=Free
  twistLimitDegrees=45
  twistMotion=Free
 }
 {bone1Name=dyn_back_cloth_01, bone2Name=base_back_cloth_01, swing1LimitDegrees=10, swing2LimitDegrees=20, twistLimitDegrees=0}
 {bone1Name=dyn_back_cloth_02, bone2Name=dyn_back_cloth_01, swing1LimitDegrees=10, swing2LimitDegrees=20, twistLimitDegrees=0}
 {bone1Name=dyn_back_cloth_03, bone2Name=dyn_back_cloth_02, swing1LimitDegrees=10, swing2LimitDegrees=20, twistLimitDegrees=0}
 {bone1Name=dyn_back_cloth_04, bone2Name=dyn_back_cloth_03, swing1LimitDegrees=10, swing2LimitDegrees=20, twistLimitDegrees=0}
 {bone1Name=dyn_back_cloth_05, bone2Name=dyn_back_cloth_04, swing1LimitDegrees=10, swing2LimitDegrees=20, twistLimitDegrees=0}
 {bone1Name=dyn_back_cloth_06, bone2Name=dyn_back_cloth_05, swing1LimitDegrees=10, swing2LimitDegrees=20, twistLimitDegrees=0}
 {bone1Name=base_back_cloth_01, bone2Name=pelvis, swing1LimitDegrees=10, swing2LimitDegrees=20, twistLimitDegrees=0}
 {
  bone1Name=calf_r
  bone2Name=thigh_r
  swing1LimitDegrees=45
  swing1Motion=Free
  swing2LimitDegrees=45
  swing2Motion=Free
  twistLimitDegrees=45
  twistMotion=Free
 }
 {
  bone1Name=calf_l
  bone2Name=thigh_l
  swing1LimitDegrees=45
  swing1Motion=Free
  swing2LimitDegrees=45
  swing2Motion=Free
  twistLimitDegrees=45
  twistMotion=Free
 }
]
