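# Animation Blueprint for the F_MED_DragonPry cosmetic body, adding secondary motion to cloth, armour pads, knife and star bones.
# It copies the parent mesh pose through a Control Rig, then chains RigidBody, AnimDynamics, constraints and additive bone offsets.
# A post Control Rig follows, and front-end idle curves blend slot animation onto the sword bones.
# The EventGraph derives gravity, movement, emote and tail state, lowering first-simulation weight when a tail is present.
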AnimBlueprint /BRCosmetics/Characters/Player/Female/Medium/Bodies/F_MED_DragonPry/Meshes/F_MED_DragonPry_AnimBP
target_skeleton=&/F_MED_DragonPry_Skeleton
graphs [
 AnimGraph {nodes [
   n1 = Control Rig!(Source=Copy Pose From Mesh() {bUseAttachedParent=true}, Alpha, CustomEmote:double=Get CustomEmote() -> (CustomEmote:double)) -> (Pose) {controlRigAssetReference {blueprintRigClass=&/F_MED_DragonPry_CtrlRig.F_MED_DragonPry_CtrlRig_C}}
   n2 = RigidBody!(
    ComponentPose=Local To Component(LocalPose=@n1)
    Alpha=Get RBNOFF() -> (RBNOFF:double)
    OverrideWorldGravity:Vector="0,0,-980"
    ExternalForce
    SimSpaceSettings:Sim Space Settings Structure=(WorldAlpha=1,VelocityScaleZ=1,DampingAlpha=1,MaxLinearVelocity=800,MaxAngularVelocity=8,MaxLinearAcceleration=4000,MaxAngularAcceleration=40,ExternalLinearDragV=(X=0.3,Y=1,Z=0),ExternalLinearVelocity=(X=0,Y=0,Z=0),ExternalAngularVelocity=(X=0,Y=0,Z=0))
   ) -> (Pose) {
    bOverrideWorldGravity=true
    overridePhysicsAsset=&/F_MED_DragonPry_Physics
    simulationSpace=BaseBoneSpace
    alphaBoolBlend {blendOption=Linear}
    baseBoneRef {boneName=pelvis}
    componentAppliedLinearAccClamp=<2000 2000 2000>
    overrideWorldGravity {z=-980}
    simSpaceSettings {
     maxAngularAcceleration=40
     maxAngularVelocity=8
     maxLinearAcceleration=4000
     maxLinearVelocity=800
     worldAlpha=1
     externalLinearDragV=<0.3 1>
    }
   }
   n3 = "Transform (Modify) Bone - Bone: dyn_back_cloth_01"(ComponentPose=@n2, Alpha=1, Translation="0,0,0", Rotation="30,0,0", Scale="1,1,1") {
    rotationMode=BMM_Additive
    rotationSpace=BCS_BoneSpace
    translationMode=BMM_Additive
    translationSpace=BCS_BoneSpace
    alphaScaleBiasClamp {bInterpResult=true, interpSpeedDecreasing=5, interpSpeedIncreasing=5}
    boneToModify {boneName=dyn_back_cloth_01}
    rotation {pitch=30}
   }
   n4 = RigidBody(
    ComponentPose=@n3
    OverrideWorldGravity="0,0,-980"
    SimSpaceSettings=(WorldAlpha=0.5,VelocityScaleZ=1,DampingAlpha=0,MaxLinearVelocity=800,MaxAngularVelocity=8,MaxLinearAcceleration=2000,MaxAngularAcceleration=40,ExternalLinearDragV=(X=0.3,Y=1,Z=0),ExternalLinearVelocity=(X=0,Y=0,Z=0),ExternalAngularVelocity=(X=0,Y=1,Z=0))
   ) {
    bOverrideWorldGravity=true
    overridePhysicsAsset=&/F_MED_DragonPry_PhysicsFrontCloth
    simulationSpace=BaseBoneSpace
    alphaBoolBlend {blendOption=Linear}
    alphaScaleBiasClamp {bInterpResult=true, interpSpeedDecreasing=3, interpSpeedIncreasing=3}
    baseBoneRef {boneName=pelvis}
    componentAppliedLinearAccClamp=<2000 2000 2000>
    overrideWorldGravity {z=-980}
    simSpaceSettings {
     dampingAlpha=0
     maxAngularAcceleration=40
     maxAngularVelocity=8
     maxLinearAcceleration=2000
     maxLinearVelocity=800
     worldAlpha=0.5
     externalAngularVelocity {y=1}
     externalLinearDragV=<0.3 1>
    }
   }
   n5 = RigidBody(
    ComponentPose=@n4
    OverrideWorldGravity="0,0,-980"
    SimSpaceSettings=(WorldAlpha=0.5,VelocityScaleZ=1,DampingAlpha=0,MaxLinearVelocity=800,MaxAngularVelocity=8,MaxLinearAcceleration=2000,MaxAngularAcceleration=40,ExternalLinearDragV=(X=0.3,Y=1,Z=0),ExternalLinearVelocity=(X=0,Y=0,Z=0),ExternalAngularVelocity=(X=0,Y=1,Z=0))
   ) {
    bOverrideWorldGravity=true
    overridePhysicsAsset=&/F_MED_DragonPry_FE_PhysicsFrontCloth
    simulationSpace=BaseBoneSpace
    alphaBoolBlend {blendOption=Linear}
    alphaScaleBiasClamp {bInterpResult=true, interpSpeedDecreasing=3, interpSpeedIncreasing=3}
    baseBoneRef {boneName=pelvis}
    componentAppliedLinearAccClamp=<2000 2000 2000>
    overrideWorldGravity {z=-980}
    simSpaceSettings {
     dampingAlpha=0
     maxAngularAcceleration=40
     maxAngularVelocity=8
     maxLinearAcceleration=2000
     maxLinearVelocity=800
     worldAlpha=0.5
     externalAngularVelocity {y=1}
     externalLinearDragV=<0.3 1>
    }
   }
   n6 = Anim Dynamics - Bone: dyn_knife_l_01!(ComponentPose=@n5, Alpha=1, GravityOverride:Vector="0,0,-980", ExternalForce="0,0,0") -> (Pose) {
    angularDampingOverride=1
    angularSpringConstant=5000
    bAngularSpring=true
    bOverrideAngularDamping=true
    bUseGravityOverride=true
    gravityScale=0.35
    simulationSpace=RootRelative
    boundBone {boneName=dyn_knife_l_01}
    componentLinearAccScale=<1 1 1>
    componentLinearVelScale=<1 1 1>
    gravityOverride {z=-980}
    physicsBodyDefinitions [{
      boundBone {boneName=dyn_knife_l_01}
      boxExtents=<20 30 30>
      constraintSetup {angularLimitsMax=<0 0 3>, angularLimitsMin=<0 0 -3>, angularTarget=<1 0 0>}
      localJointOffset=<4 0 0>
     }]
   }
   n7 = "Constraint - dyn_star_r_01 (PPP)"!(ComponentPose=@n6, Alpha=1, ConstraintWeights_0=0.5195, ConstraintWeights_1:float=0.4654, ConstraintWeights_2:float=0.0152) -> (Pose) {
    boneToModify {boneName=dyn_star_r_01}
    constraintSetup [{targetBone {boneName=thigh_twist_01_r}}, {targetBone {boneName=thigh_r}}, {targetBone {boneName=pelvis}}]
    constraintWeights [1,1,1]
   }
   n8 = "Constraint - dyn_star_r_02 (PP)"!(ComponentPose=@n7, Alpha=1, ConstraintWeights_0=0.7604, ConstraintWeights_1:float=0.2396) -> (Pose) {
    boneToModify {boneName=dyn_star_r_02}
    constraintSetup [{targetBone {boneName=thigh_r}}, {targetBone {boneName=thigh_twist_01_r}}]
    constraintWeights [1,1]
   }
   n9 = "Constraint - base_knife_l_01 (PP)"!(ComponentPose=@n8, Alpha=1, ConstraintWeights_0=0.5438, ConstraintWeights_1:float=0.4562) -> (Pose) {
    boneToModify {boneName=base_knife_l_01}
    constraintSetup [{targetBone {boneName=thigh_l}}, {targetBone {boneName=thigh_twist_01_l}}]
    constraintWeights [1,1]
   }
   n10 = Anim Dynamics - Bone: dyn_star_r_01!(ComponentPose=@n9, Alpha=1, GravityOverride:Vector="0,0,-980", ExternalForce="0,0,0") -> (Pose) {
    angularDampingOverride=1
    angularSpringConstant=5000
    bAngularSpring=true
    bOverrideAngularDamping=true
    bUseGravityOverride=true
    gravityScale=0.35
    simulationSpace=RootRelative
    boundBone {boneName=dyn_star_r_01}
    componentLinearAccScale=<1 1 1>
    componentLinearVelScale=<1 1 1>
    gravityOverride {z=-980}
    physicsBodyDefinitions [{
      boundBone {boneName=dyn_star_r_01}
      boxExtents=<20 30 30>
      constraintSetup {angularLimitsMax=<0 1 5>, angularLimitsMin=<0 -1 -5>, angularTarget=<1 0 0>}
      localJointOffset=<1 0 0>
     }]
   }
   n11 = Anim Dynamics - Bone: dyn_star_r_02!(ComponentPose=@n10, Alpha=1, GravityOverride:Vector="0,0,-980", ExternalForce="0,0,0") -> (Pose) {
    angularDampingOverride=1
    angularSpringConstant=5000
    bAngularSpring=true
    bOverrideAngularDamping=true
    bUseGravityOverride=true
    gravityScale=0.35
    simulationSpace=RootRelative
    boundBone {boneName=dyn_star_r_02}
    componentLinearAccScale=<1 1 1>
    componentLinearVelScale=<1 1 1>
    gravityOverride {z=-980}
    physicsBodyDefinitions [{
      boundBone {boneName=dyn_star_r_02}
      boxExtents=<20 30 30>
      constraintSetup {angularLimitsMax=<0 1 5>, angularLimitsMin=<0 -1 -5>, angularTarget=<1 0 0>}
      localJointOffset=<1 0 0>
     }]
   }
   n12 = "Transform (Modify) Bone - Bone: armourPad_r"!(ComponentPose=@n11, bAlphaBoolEnabled:bool=True, Translation="1.3,0,0.244", Rotation="10.9,0,0", Scale="1,1,1") -> (Pose) {
    alphaInputType=Bool
    rotationMode=BMM_Additive
    rotationSpace=BCS_BoneSpace
    scaleMode=BMM_Additive
    scaleSpace=BCS_BoneSpace
    translationMode=BMM_Additive
    translationSpace=BCS_BoneSpace
    alphaScaleBiasClamp {bInterpResult=true, interpSpeedDecreasing=5, interpSpeedIncreasing=3}
    boneToModify {boneName=armourPad_r}
    rotation {pitch=10.9}
    translation {x=1.3, z=0.244}
   }
   n13 = "Transform (Modify) Bone - Bone: armourPad_l"!(ComponentPose=@n12, bAlphaBoolEnabled:bool=True, Translation="1.3,0,0", Rotation="-10.9,0,0", Scale="1,1,1") -> (Pose) {
    alphaInputType=Bool
    rotationMode=BMM_Additive
    rotationSpace=BCS_BoneSpace
    scaleMode=BMM_Additive
    scaleSpace=BCS_BoneSpace
    translationMode=BMM_Additive
    translationSpace=BCS_BoneSpace
    alphaScaleBiasClamp {bInterpResult=true, interpSpeedDecreasing=5, interpSpeedIncreasing=3}
    boneToModify {boneName=armourPad_l}
    rotation {pitch=-10.9}
    translation {x=1.3}
   }
   n14 = "Transform (Modify) Bone - Bone: armourPad_r"(ComponentPose=@n13, bAlphaBoolEnabled=True, Translation="1.3,0,-2", Rotation="0,0,0", Scale="1,1,1") {
    alphaInputType=Bool
    rotationMode=BMM_Additive
    rotationSpace=BCS_BoneSpace
    scaleMode=BMM_Additive
    scaleSpace=BCS_BoneSpace
    translationMode=BMM_Additive
    translationSpace=BCS_BoneSpace
    alphaScaleBiasClamp {bInterpResult=true, interpSpeedDecreasing=5, interpSpeedIncreasing=3}
    boneToModify {boneName=armourPad_r}
    translation {x=1.3, z=-2}
   }
   n15 = "Transform (Modify) Bone - Bone: armourPad_l"(ComponentPose=@n14, bAlphaBoolEnabled=True, Translation="1.3,0,2", Rotation="0,0,0", Scale="1,1,1") {
    alphaInputType=Bool
    rotationMode=BMM_Additive
    rotationSpace=BCS_BoneSpace
    scaleMode=BMM_Additive
    scaleSpace=BCS_BoneSpace
    translationMode=BMM_Additive
    translationSpace=BCS_BoneSpace
    alphaScaleBiasClamp {bInterpResult=true, interpSpeedDecreasing=5, interpSpeedIncreasing=3}
    boneToModify {boneName=armourPad_l}
    translation {x=1.3, z=2}
   }
   n16 = "Transform (Modify) Bone - Bone: armourPad_r"(ComponentPose=@n15, bAlphaBoolEnabled=True, Translation="-0.359,-0.016,0.486", Rotation="2.343,2.719,-4.436", Scale="1,1,1") {
    alphaInputType=Bool
    rotationMode=BMM_Additive
    rotationSpace=BCS_BoneSpace
    scaleMode=BMM_Additive
    scaleSpace=BCS_BoneSpace
    translationMode=BMM_Additive
    translationSpace=BCS_BoneSpace
    alphaScaleBiasClamp {bInterpResult=true, interpSpeedDecreasing=5, interpSpeedIncreasing=3}
    boneToModify {boneName=armourPad_r}
    rotation=<rot 2.343 2.719 -4.436>
    translation=<-0.359 -0.016 0.486>
   }
   n17 = "Transform (Modify) Bone - Bone: armourPad_l"(ComponentPose=@n16, bAlphaBoolEnabled=True, Translation="0.563,-0.000,-2.532", Rotation="-8.836,-0.000,0.000", Scale="1,1,1") {
    alphaInputType=Bool
    rotationMode=BMM_Additive
    rotationSpace=BCS_BoneSpace
    scaleMode=BMM_Additive
    scaleSpace=BCS_BoneSpace
    translationMode=BMM_Additive
    translationSpace=BCS_BoneSpace
    alphaScaleBiasClamp {bInterpResult=true, interpSpeedDecreasing=5, interpSpeedIncreasing=3}
    boneToModify {boneName=armourPad_l}
    rotation=<rot -8.836 -0.0001 0>
    translation {x=0.563, z=-2.532}
   }
   n18 = "Transform (Modify) Bone - Bone: dyn_front_cloth_mid_01"!(ComponentPose=@n17, bAlphaBoolEnabled:bool=True, Translation="0,0,0", Rotation="5,0,0", Scale="1,1,1") -> (Pose) {
    alphaInputType=Bool
    rotationMode=BMM_Additive
    rotationSpace=BCS_BoneSpace
    scaleMode=BMM_Additive
    scaleSpace=BCS_BoneSpace
    translationMode=BMM_Additive
    translationSpace=BCS_BoneSpace
    alphaBoolBlend {blendOption=Linear}
    alphaScaleBiasClamp {bInterpResult=true, interpSpeedDecreasing=5, interpSpeedIncreasing=3}
    boneToModify {boneName=dyn_front_cloth_mid_01}
    rotation {pitch=5}
   }
   n19 = "Transform (Modify) Bone - Bone: dyn_front_cloth_mid_01"(ComponentPose=@n18, bAlphaBoolEnabled=True, Translation="0,0,0", Rotation="20,0,0", Scale="1,1,1") {
    alphaInputType=Bool
    rotationMode=BMM_Additive
    rotationSpace=BCS_BoneSpace
    scaleMode=BMM_Additive
    scaleSpace=BCS_BoneSpace
    translationMode=BMM_Additive
    translationSpace=BCS_BoneSpace
    alphaBoolBlend {blendInTime=0.1, blendOption=Linear, blendOutTime=0.1}
    alphaScaleBiasClamp {bInterpResult=true, interpSpeedDecreasing=5, interpSpeedIncreasing=3}
    boneToModify {boneName=dyn_front_cloth_mid_01}
    rotation {pitch=20}
   }
   n20 = "Transform (Modify) Bone - Bone: base_back_cloth_01"!(ComponentPose=@n19, bAlphaBoolEnabled:bool=True, Translation="0,0,0", Rotation="0,0,0", Scale="1,1,1") -> (Pose) {
    alphaInputType=Bool
    rotationMode=BMM_Additive
    rotationSpace=BCS_BoneSpace
    scaleMode=BMM_Additive
    scaleSpace=BCS_BoneSpace
    translationMode=BMM_Additive
    translationSpace=BCS_BoneSpace
    alphaBoolBlend {blendInTime=0.1, blendOption=Linear, blendOutTime=0.1}
    alphaScaleBiasClamp {bInterpResult=true, interpSpeedDecreasing=5, interpSpeedIncreasing=3}
    boneToModify {boneName=base_back_cloth_01}
   }
   n21 = "Transform (Modify) Bone - Bone: dyn_front_cloth_mid_01"(ComponentPose=@n20, bAlphaBoolEnabled=True, Translation="0,0,0", Rotation="-9.117,-0.000,0.000", Scale="1,1,1") {
    alphaInputType=Bool
    rotationMode=BMM_Additive
    rotationSpace=BCS_BoneSpace
    scaleMode=BMM_Additive
    scaleSpace=BCS_BoneSpace
    translationMode=BMM_Additive
    translationSpace=BCS_BoneSpace
    alphaBoolBlend {blendInTime=0.1, blendOption=Linear, blendOutTime=0.1}
    alphaScaleBiasClamp {bInterpResult=true, interpSpeedDecreasing=5, interpSpeedIncreasing=3}
    boneToModify {boneName=dyn_front_cloth_mid_01}
    rotation=<rot -9.117 0 0>
   }
   n22 = "Transform (Modify) Bone - Bone: dyn_front_cloth_mid_02"!(ComponentPose=@n21, bAlphaBoolEnabled:bool=True, Translation="0,0,0", Rotation="10.600,-0.000,0.000", Scale="1,1,1") -> (Pose) {
    alphaInputType=Bool
    rotationMode=BMM_Additive
    rotationSpace=BCS_BoneSpace
    scaleMode=BMM_Additive
    scaleSpace=BCS_BoneSpace
    translationMode=BMM_Additive
    translationSpace=BCS_BoneSpace
    alphaBoolBlend {blendInTime=0.1, blendOption=Linear, blendOutTime=0.1}
    alphaScaleBiasClamp {bInterpResult=true, interpSpeedDecreasing=5, interpSpeedIncreasing=3}
    boneToModify {boneName=dyn_front_cloth_mid_02}
    rotation=<rot 10.6 0 0>
   }
   n23 = "Transform (Modify) Bone - Bone: dyn_front_cloth_mid_03"!(ComponentPose=@n22, bAlphaBoolEnabled:bool=True, Translation="0,0,0", Rotation="23.025,-0.000,0.000", Scale="1,1,1") -> (Pose) {
    alphaInputType=Bool
    rotationMode=BMM_Additive
    rotationSpace=BCS_BoneSpace
    scaleMode=BMM_Additive
    scaleSpace=BCS_BoneSpace
    translationMode=BMM_Additive
    translationSpace=BCS_BoneSpace
    alphaBoolBlend {blendInTime=0.1, blendOption=Linear, blendOutTime=0.1}
    alphaScaleBiasClamp {bInterpResult=true, interpSpeedDecreasing=5, interpSpeedIncreasing=3}
    boneToModify {boneName=dyn_front_cloth_mid_03}
    rotation=<rot 23.025 0 0>
   }
   n24 = "Transform (Modify) Bone - Bone: dyn_front_cloth_mid_04"!(ComponentPose=@n23, bAlphaBoolEnabled:bool=True, Translation="0,0,0", Rotation="25.518,-0.000,0.000", Scale="1,1,1") -> (Pose) {
    alphaInputType=Bool
    rotationMode=BMM_Additive
    rotationSpace=BCS_BoneSpace
    scaleMode=BMM_Additive
    scaleSpace=BCS_BoneSpace
    translationMode=BMM_Additive
    translationSpace=BCS_BoneSpace
    alphaBoolBlend {blendInTime=0.1, blendOption=Linear, blendOutTime=0.1}
    alphaScaleBiasClamp {bInterpResult=true, interpSpeedDecreasing=5, interpSpeedIncreasing=3}
    boneToModify {boneName=dyn_front_cloth_mid_04}
    rotation=<rot 25.518 0 0>
   }
   n25 = "Transform (Modify) Bone - Bone: dyn_front_cloth_mid_05"!(ComponentPose=@n24, bAlphaBoolEnabled:bool=True, Translation="0,0,0", Rotation="20,0,0", Scale="1,1,1") -> (Pose) {
    alphaInputType=Bool
    rotationMode=BMM_Additive
    rotationSpace=BCS_BoneSpace
    scaleMode=BMM_Additive
    scaleSpace=BCS_BoneSpace
    translationMode=BMM_Additive
    translationSpace=BCS_BoneSpace
    alphaBoolBlend {blendInTime=0.1, blendOption=Linear, blendOutTime=0.1}
    alphaScaleBiasClamp {bInterpResult=true, interpSpeedDecreasing=5, interpSpeedIncreasing=3}
    boneToModify {boneName=dyn_front_cloth_mid_05}
    rotation {pitch=20}
   }
   n26 = "Transform (Modify) Bone - Bone: base_back_cloth_01"!(ComponentPose=@n25, Alpha:float=1, Translation="0,0,0", Rotation="52.079,-2.261,8.122", Scale="1,1,1") -> (Pose) {
    rotationMode=BMM_Additive
    rotationSpace=BCS_BoneSpace
    translationMode=BMM_Additive
    translationSpace=BCS_BoneSpace
    boneToModify {boneName=base_back_cloth_01}
    rotation=<rot 52.079 -2.261 8.122>
   }
   n27 = "Transform (Modify) Bone - Bone: dyn_back_cloth_01"(ComponentPose=@n26, Alpha=1, Translation="0,0,0", Rotation="-5.136,-2.903,4.304", Scale="1,1,1") {
    rotationMode=BMM_Additive
    rotationSpace=BCS_BoneSpace
    translationMode=BMM_Additive
    translationSpace=BCS_BoneSpace
    boneToModify {boneName=dyn_back_cloth_01}
    rotation=<rot -5.136 -2.903 4.304>
   }
   n28 = "Transform (Modify) Bone - Bone: dyn_back_cloth_02"(ComponentPose=@n27, Alpha=1, Translation="0,0,0", Rotation="-6.961,0.000,-0.000", Scale="1,1,1") {
    rotationMode=BMM_Additive
    rotationSpace=BCS_BoneSpace
    translationMode=BMM_Additive
    translationSpace=BCS_BoneSpace
    boneToModify {boneName=dyn_back_cloth_02}
    rotation=<rot -6.961 0 0>
   }
   n29 = "Transform (Modify) Bone - Bone: base_back_cloth_01"(ComponentPose=@n28, Alpha=1, Translation="0,0,0", Rotation="16.236,-1.373,17.743", Scale="1,1,1") {
    rotationMode=BMM_Additive
    rotationSpace=BCS_BoneSpace
    translationMode=BMM_Additive
    translationSpace=BCS_BoneSpace
    boneToModify {boneName=base_back_cloth_01}
    rotation=<rot 16.236 -1.373 17.743>
   }
   n30 = "Transform (Modify) Bone - Bone: dyn_back_cloth_01"(ComponentPose=@n29, Alpha=1, Translation="0,0,0", Rotation="15,0,0", Scale="1,1,1") {
    rotationMode=BMM_Additive
    rotationSpace=BCS_BoneSpace
    translationMode=BMM_Additive
    translationSpace=BCS_BoneSpace
    boneToModify {boneName=dyn_back_cloth_01}
    rotation {pitch=15}
   }
   n31 = "Transform (Modify) Bone - Bone: dyn_back_cloth_02"(ComponentPose=@n30, Alpha=1, Translation="0,0,0", Rotation="-20.894,-0.000,-0.000", Scale="1,1,1") {
    rotationMode=BMM_Additive
    rotationSpace=BCS_BoneSpace
    translationMode=BMM_Additive
    translationSpace=BCS_BoneSpace
    boneToModify {boneName=dyn_back_cloth_02}
    rotation=<rot -20.894 0 -0.0001>
   }
   n32 = RigidBody(
    ComponentPose=@n31
    OverrideWorldGravity="0,0,-980"
    SimSpaceSettings=(WorldAlpha=1,VelocityScaleZ=1,DampingAlpha=1,MaxLinearVelocity=800,MaxAngularVelocity=8,MaxLinearAcceleration=4000,MaxAngularAcceleration=40,ExternalLinearDragV=(X=0.3,Y=1,Z=0),ExternalLinearVelocity=(X=0,Y=0,Z=0),ExternalAngularVelocity=(X=0,Y=0,Z=0))
   ) {
    bOverrideWorldGravity=true
    overridePhysicsAsset=&/F_MED_DragonPry_Physics_FE
    simulationSpace=BaseBoneSpace
    alphaBoolBlend {blendOption=Linear}
    baseBoneRef {boneName=pelvis}
    componentAppliedLinearAccClamp=<2000 2000 2000>
    overrideWorldGravity {z=-980}
    simSpaceSettings {
     maxAngularAcceleration=40
     maxAngularVelocity=8
     maxLinearAcceleration=4000
     maxLinearVelocity=800
     worldAlpha=1
     externalLinearDragV=<0.3 1>
    }
   }
   n33 = RigidBody(
    ComponentPose=@n32
    OverrideWorldGravity="0,0,-980"
    SimSpaceSettings=(WorldAlpha=0.5,VelocityScaleZ=1,DampingAlpha=0,MaxLinearVelocity=800,MaxAngularVelocity=8,MaxLinearAcceleration=2000,MaxAngularAcceleration=40,ExternalLinearDragV=(X=0.3,Y=1,Z=0),ExternalLinearVelocity=(X=0,Y=0,Z=0),ExternalAngularVelocity=(X=0,Y=1,Z=0))
   ) {
    bOverrideWorldGravity=true
    overridePhysicsAsset=&/F_MED_DragonPry_PhysicsFrontCloth_FE
    simulationSpace=BaseBoneSpace
    alphaBoolBlend {blendOption=Linear}
    baseBoneRef {boneName=pelvis}
    componentAppliedLinearAccClamp=<2000 2000 2000>
    overrideWorldGravity {z=-980}
    simSpaceSettings {
     dampingAlpha=0
     maxAngularAcceleration=40
     maxAngularVelocity=8
     maxLinearAcceleration=2000
     maxLinearVelocity=800
     worldAlpha=0.5
     externalAngularVelocity {y=1}
     externalLinearDragV=<0.3 1>
    }
   }
   n34 = "Transform (Modify) Bone - Bone: base_front_cloth_m_01"(ComponentPose=@n33, Alpha=1, Translation="0,0.2,1", Rotation="-5,0,10", Scale="1,1,1") {
    rotationMode=BMM_Additive
    rotationSpace=BCS_BoneSpace
    translationMode=BMM_Additive
    translationSpace=BCS_BoneSpace
    boneToModify {boneName=base_front_cloth_m_01}
    rotation {pitch=-5, roll=10}
    translation {y=0.2, z=1}
   }
   n35 = "Transform (Modify) Bone - Bone: dyn_back_cloth_02"(ComponentPose=@n34, Alpha=1, Translation="0,0,0", Rotation="5,0,0", Scale="1,1,1") {
    rotationMode=BMM_Additive
    rotationSpace=BCS_BoneSpace
    translationMode=BMM_Additive
    translationSpace=BCS_BoneSpace
    boneToModify {boneName=dyn_back_cloth_02}
    rotation {pitch=5}
   }
   n36 = "Transform (Modify) Bone - Bone: base_back_cloth_01"(
    ComponentPose=@n35
    Alpha=Get Skydiving_Float_Alpha() -> (Skydiving_Float_Alpha:double)
    Translation="0,0,1.2"
    Rotation="-20,0,0"
    Scale="1,1,1"
   ) {
    rotationMode=BMM_Additive
    rotationSpace=BCS_BoneSpace
    translationMode=BMM_Additive
    translationSpace=BCS_BoneSpace
    alphaScaleBiasClamp {bInterpResult=true, interpSpeedDecreasing=2, interpSpeedIncreasing=1}
    boneToModify {boneName=base_back_cloth_01}
    rotation {pitch=-20}
    translation {z=1.2}
   }
   n37 = Control Rig(Source=Component To Local(ComponentPose=@n36), CustomEmote=Get CustomEmote()) {controlRigAssetReference {blueprintRigClass=&/F_MED_DragonPry_post_CtrlRig.F_MED_DragonPry_post_CtrlRig_C}}
   n38 = Layered blend per bone(BasePose=@n37, BlendPoses_0=Slot 'DefaultSlot'(), BlendWeights_0=Get Custom Idle Is Playing Float() -> (Custom Idle Is Playing Float:double)) {layerSetup [{branchFilters [{boneName=base_sword_weapon_01}, {boneName=dyn_sword_weapon_01}]}]}
   Output Pose(Result=@n38) {layerGroup=None}
  ]}
 EventGraph {nodes [
   Event BlueprintUpdateAnimation()
   > n1 = Sequence!(execute) -> (then_0, then_1, then_2:Exec, then_3:Exec)
   > n2 = "Gravity Override State Function (Parameter Asset) (Nativized)"(
    CharacterPartAnimInstance=Self-Reference()
    GameTimeInSeconds=GetGameTimeInSeconds()
    Parameters=&/F_MED_Dragon_Pry_Parameters
   )
   > Set GravityOverride(GravityOverride=@n2)
   > Set RBAN Settings Out(RBAN Settings Out=@n2.OutRigidBodyAnimNodeParameters)
   > Branch(Condition=Get bIsInFrontEnd())
   > Set Custom Idle Is Playing Float(Custom Idle Is Playing Float:double=GetCurveValue(self=GetOwnerFrontEndAnimInstance(), CurveName=Frontend_Idle_Bypass)) -> (Output_Get:double)
   n3 = Get bIsPlayingEmote()
   > n4 = Branch(Condition=AND Boolean(A=Get bIsInFrontEnd(), B=NOT Boolean(A=@n3)))
   > Set Is Front End(Is Front End:double=1) -> (Output_Get:double)
   Set Is Front End(execute=@n4.else, Is Front End=0)
   Set Crouchwalk(execute=[@n1.then_1, @n1.then_2], Crouchwalk:bool=AND Boolean(A=Get bIsMoving2D(), B=Get bIsCrouching())) -> (Output_Get:bool)
   n5 = GetAnimInstance(self=GetOwnerSkeletalMeshComponent())
   n6 = Montage_IsPlaying(self=@n5, Montage=/BRCosmetics/Animation/Game/MainPlayer/Emotes/Cobbler/CMF/Emote_Cobbler_CMF_M)
   n7 = Montage_IsPlaying(self=@n5, Montage=/BRCosmetics/Animation/Game/MainPlayer/Emotes/Emote_Celebration_Female_M)
   > Set CustomEmote(CustomEmote:double="To Float (Boolean)"(InBool=OR Boolean(A=@n7, B=@n6))) -> (Output_Get:double)
   n8 = CheckForTail!(self:Custom Character Part Anim Instance Object Reference) -> (bHasTail:bool)
   n9 = Get bIsBaseballSliding()
   > Set AR_Sliding(AR_Sliding:double="To Float (Boolean)"(InBool=AND Boolean(A=@n9, B=@n8))) -> (Output_Get:double)
   n10 = Get bIsSliding()
   > Set Sliding(Sliding:double="To Float (Boolean)"(InBool=AND Boolean(A=@n10, B=@n8))) -> (Output_Get:double)
   > n11 = Branch(Condition=Get bIsTacticalSprinting())
   > Set Sprinting(Sprinting:double=1) -> (Output_Get:double)
   Set Sprinting(execute=@n11.else, Sprinting=0)
   n12 = Branch(execute=@n1.then_3, Condition=AND Boolean(A=Get bIsInFrontEnd(), B=Get bIsPlayingEmote()))
   > Set Frontend_Emote_DYN(Frontend_Emote_DYN:double=0.7) -> (Output_Get:double)
   Set Frontend_Emote_DYN(execute=@n12.else, Frontend_Emote_DYN=0)
   Event OnCharacterCustomizationLoaded()
   > n13 = Set TailOnOff(TailOnOff:bool=CheckForTail()) -> (Output_Get:bool)
   > n14 = Branch(Condition=@n13)
   > Set RBNOFF(RBNOFF:double=0.65) -> (Output_Get:double)
   Set RBNOFF(execute=@n14.else, RBNOFF=1)
   n15 = Branch(Condition=Get bIsSkydiveFloating())
   > Set Skydiving_Float_Alpha(Skydiving_Float_Alpha:double=1) -> (Output_Get:double)
   Set Skydiving_Float_Alpha(execute=@n15.else, Skydiving_Float_Alpha=0)
  ]}
]
variables [
 AR_Sliding {type=number}
 Crouchwalk {type=boolean}
 Custom Idle Is Playing Float {type=number}
 CustomEmote {type=number}
 Delta Time X {type=number}
 Frontend_Emote_DYN {type=number}
 GravityOverride {type=Vector}
 Is Front End {type=number}
 RBAN Settings Out {type=FortRigidBodyAnimNodeParameters}
 RBNOFF {type=number}
 RBNOFF_Tail {type=number}
 Skydiving_Float {type=number}
 Skydiving_Float_Alpha {type=number}
 Sliding {type=number}
 Sprinting {type=number}
 TailOnOff {type=boolean}
]
